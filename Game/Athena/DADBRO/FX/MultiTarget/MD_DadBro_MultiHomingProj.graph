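# Translucent deferred decal material for DadBro's multi-target homing projectile effect, apparently a circular marker.
# A SphereMask on the texture coordinates, centered at 0.5, forms a circle with Radius and Hardness parameters.
# An Alpha parameter scales that mask, alongside two standalone constants of zero and one.
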Material /Game/Athena/DADBRO/FX/MultiTarget/MD_DadBro_MultiHomingProj
graph {
 properties {blendMode=BLEND_Translucent, materialDomain=MD_DeferredDecal, pixelDepthOffsetMode=PDOM_Legacy, usedShadingModels=Default Lit}
 nodes [
  Constant()
  Constant() {r=1}
  Multiply(
   A=SphereMask(
    A=TextureCoordinate()
    B=Constant() {r=0.5}
    Radius=ScalarParameter() {defaultValue=0.45, parameterName=Radius}
    Hardness=ScalarParameter() {parameterName=Hardness}
   )
   B=ScalarParameter() {defaultValue=0.85, parameterName=Alpha}
  )
 ]
}
